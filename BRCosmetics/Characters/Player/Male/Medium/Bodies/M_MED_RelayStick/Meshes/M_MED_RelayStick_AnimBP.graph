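# Animation Blueprint for the M_MED_RelayStick player body that copies the pose from the attached parent mesh and refines it with a deform Control Rig.
# Each update casts the owning mesh component's anim instance to FortPlayerAnimInstance and switches on its weapon core animation.
# Only AssaultRifle and Pistol set flags, which become floats driving the rig's LargeGunEquipped and PistolEquipped inputs.
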AnimBlueprint /BRCosmetics/Characters/Player/Male/Medium/Bodies/M_MED_RelayStick/Meshes/M_MED_RelayStick_AnimBP
target_skeleton=&/M_MED_RelayStick_Skeleton
graphs [
 AnimGraph {nodes [
   Output Pose(Result=Use cached pose 'CRCache_Input'()) {layerGroup=None}
   n1 = Control Rig!(
    Source=Copy Pose From Mesh() {bUseAttachedParent=true}
    Alpha
    LargeGunEquipped:double=Get large_gun_out() -> (large_gun_out:double)
    PistolEquipped:double=Get PistolGunOut() -> (PistolGunOut:double)
   ) -> (Pose) {
    alphaBoolBlend {blendOption=Linear}
    controlRigAssetReference {blueprintRigClass=&/M_MED_DeformRig_RelayStick_ControlRig_new.M_MED_DeformRig_RelayStick_ControlRig_new_C}
   }
   Save cached pose 'CRCache_Input'(Pose=@n1) {cacheName=CRCache_Input}
  ]}
 EventGraph {nodes [
   Event BlueprintUpdateAnimation()
   > n1 = Sequence()
   > n2 = Cast To FortPlayerAnimInstance(Object=GetAnimInstance(self=GetOwnerSkeletalMeshComponent()))
   > Set rifle_out(rifle_out:bool=false) -> (Output_Get:bool)
   > Set PistolOut(PistolOut:bool=false) -> (Output_Get:bool)
   > n3 = Switch on EFortWeaponCoreAnimation!(execute, Selection:EFortWeaponCoreAnimation Enum=Get WeaponCoreAnim(self=@n2)) -> (
    Melee:Exec
    Pistol:Exec
    Shotgun:Exec
    PaperBlueprint:Exec
    Rifle:Exec
    MeleeOneHand:Exec
    MachinePistol:Exec
    RocketLauncher:Exec
    GrenadeLauncher:Exec
    GoingCommando:Exec
    AssaultRifle:Exec
    TacticalShotgun:Exec
    SniperRifle:Exec
    TrapPlacement:Exec
    ShoulderLauncher:Exec
    AbilityDecoTool:Exec
    Crossbow:Exec
    C4:Exec
    RemoteControl:Exec
    DualWield:Exec
    AR_BullPup:Exec
    AR_ForwardGrip:Exec
    MedPackPaddles:Exec
    SMG_P90:Exec
    AR_DrumGun:Exec
    Consumable_Small:Exec
    Consumable_Large:Exec
    Balloon:Exec
    MountedTurret:Exec
    CreativeTool:Exec
    ExplosiveBow:Exec
    AshtonIndigo:Exec
    AshtonChicago:Exec
    MeleeDualWield:Exec
    Unarmed:Exec
   )
   Branch(execute=@n3.AssaultRifle)
   > Set rifle_out(rifle_out=true)
   Branch(execute=@n3.Pistol)
   > Set PistolOut(PistolOut=true)
   Set large_gun_out(execute=@n1.then_1, large_gun_out:double="To Float (Boolean)"(InBool=Get rifle_out() -> (rifle_out:bool))) -> (Output_Get:double)
   > Set PistolGunOut(PistolGunOut:double="To Float (Boolean)"(InBool=Get PistolOut() -> (PistolOut:bool))) -> (Output_Get:double)
   Reroute Node(InputPin:Wildcard) -> (OutputPin:Wildcard)
  ]}
]
variables [
 large_gun_out {type=number}
 PistolGunOut {type=number}
 PistolOut {type=boolean}
 rifle_out {type=boolean}
]
